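# Niagara effect for the cap spin in the player's Bottle Cap Challenge emote, built from three mesh emitters.
# Big_Glow_Mesh pops one brief yellow-green glow that shrinks and fades.
# Flash bursts four very short-lived, randomly oriented meshes with panning material; its light renderer remains, but the light module is disabled.
# Swirls_Mesh continuously spawns rotating, slowly rising meshes along a vertical range, fading them through alpha and dynamic material curves.
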NiagaraSystem /Game/Animation/Game/MainPlayer/Emotes/BottleCap_Challenge/VFX/NS_BottleCap_Challenge_CapSpin
dependencies {
 usedDataInterfaces [&NiagaraDataInterfaceCurve, &NiagaraDataInterfaceVectorCurve]
 usedDynamicInputs []
 usedModules [/Niagara/Modules/System/SystemLifeCycle, &SystemState]
}
emitters {
 Big_Glow_Mesh {
  emitterUpdateScript [
   &EmitterScalability(
    Module.bApplySpawnCountScale=false
    Module.bApplySpawnCountScaleByDistanceFraction=false
    Module.bUseMaxDistance=false
    Module.bUseMinDistance=false
   ) {enabled=false}
   &EmitterLifeCycle(NextLoopDuration=1) {enabled=false}
   &SpawnRate(SpawnRate=20, Use Spawn Probability=false) {enabled=false}
   &EmitterState()
   &SpawnBurst_Instantaneous(Spawn Count=1, Use Spawn Probability=false)
  ]
  particleSpawnScript [/Niagara/Modules/Spawn/Initialization/InitializeParticle(
    Color=<rgba 2 1.1793 0.1912 1>
    Lifetime=0.1
    Mesh Scale=<0.65 0.65 0.65>
    Module.Write Color=true
    Module.Write Lifetime=true
    Module.Write Mass=true
    Module.Write Material Random=true
    Module.Write Position=true
    Module.Write Scale=true
    Module.Write SpriteRotation=false
    Module.Write SpriteSize=false
    Position=&SimulationPosition()
    Sprite Size=<25 25>
   )]
  particleUpdateScript [
   &ParticleState()
   &ScaleColor(
    Scale Alpha=&FloatFromCurve(FloatCurve=DataInterface() {propertyValues {Curve {keys [{arriveTangent=-3.2421, interpMode=RCIM_Cubic, leaveTangent=-3.2421, tangentMode=RCTM_User, time=-0.0046, value=1.0113}, <at 1.0011 0.002>]}}})
    Scale RGB=&MakeVectorFromLinearColorRGB(Color=<rgba 1 1 1 1>)
    ScaleA=true
    ScaleRGB=true
   )
   &ScaleMeshSize(Scale Factor=&VectorFromFloat(Value=&FloatFromCurve(FloatCurve=DataInterface() {propertyValues {Curve {keys [
          {arriveTangent=39.6971, interpMode=RCIM_Cubic, leaveTangent=39.6972, tangentMode=RCTM_User, time=0, value=0}
          {arriveTangent=-0.2033, interpMode=RCIM_Cubic, leaveTangent=-0.2033, tangentMode=RCTM_User, time=0.05, value=1}
          {arriveTangent=-2.1178, interpMode=RCIM_Cubic, leaveTangent=-2.1178, tangentMode=RCTM_User, time=1.0011, value=0.002}
         ]}}})))
  ]
  rendererClasses [&NiagaraMeshRendererProperties]
  renderers [{rendererClass=&NiagaraMeshRendererProperties, rendererIndex=0}]
 }
 Flash {
  emitterUpdateScript [
   &EmitterScalability(
    Module.bApplySpawnCountScale=false
    Module.bApplySpawnCountScaleByDistanceFraction=false
    Module.bUseMaxDistance=false
    Module.bUseMinDistance=false
   ) {enabled=false}
   &EmitterLifeCycle() {enabled=false}
   &EmitterState()
   &SpawnBurst_Instantaneous(Spawn Count=4, Use Spawn Probability=false)
  ]
  particleSpawnScript [
   &InitializeParticle(
    Color=&MakeLinearColor(
     R:NiagaraFloat=1
     G:NiagaraFloat=1
     B:NiagaraFloat=1
     A:NiagaraFloat=&UniformRangedFloat(Minimum=0.1, Maximum=0.2)
    )
    Lifetime=&UniformRangedFloat(Minimum=0.05, Maximum=0.05)
    Mesh Scale=&UniformRangedVector(Minimum=<20 20 20>, Maximum=<3 3 3>)
    Module.Write Color=true
    Module.Write Lifetime=true
    Module.Write Mass=true
    Module.Write Material Random=true
    Module.Write Position=true
    Module.Write Scale=true
    Module.Write SpriteRotation=false
    Module.Write SpriteSize=false
    Position=&Add_Vector(A=/Niagara/DynamicInputs/Helpers/SimulationPosition(), B=<0 0 0>)
   )
   /SetVariables(Particles.Scale=&UniformRangedVector(Minimum=<0.6 0.4 0.4>, Maximum=<0.4 0.7 0.7>)) {isSetParameters=true}
   /Game/Effects/Niagara/Modules/Orientation/FN_Fan(
    Cone Angle:NiagaraFloat=&UniformRangedFloat(Minimum=0, Maximum=0)
    Spin:NiagaraFloat=&UniformRangedFloat(Minimum=0, Maximum=6.2)
   ) {enabled=false}
   &ApplyInitialForces() {enabled=false}
   /Niagara/Modules/Spawn/Orientation/V2/InitialMeshOrientation(
    Orientation Coordinate Space:ENiagaraCoordinateSpace=::Local
    Orientation Vector=&UniformRangedVector(Minimum=<-0.15 0 -0.5>, Maximum=<-0.1 -1 0.5>)
    Use Orientation Vector=true
    Use Rotation Vector=false
   )
  ]
  particleUpdateScript [
   &ParticleState()
   &Color(Scale Color=&VectorFromCurve(VectorCurve=DataInterface() {propertyValues {
       XCurve {keys [<at -0.0031 48.8>,<at 1 1>]}
       YCurve {keys [<at -0.006 20.3>,<at 1 0.8>]}
       ZCurve {keys [<at -0.0011 1.0067>,<at 0.9989 0.28>]}
      }})) {enabled=false}
   &ScaleMeshSize(Scale Factor=&VectorFromCurve(VectorCurve=DataInterface() {propertyValues {
       XCurve {keys [{arriveTangent=6.3957, interpMode=RCIM_Cubic, leaveTangent=6.3957, tangentMode=RCTM_User, time=0, value=0.47}, <at 0.9748 1.29>]}
       YCurve {keys [{arriveTangent=6.1645, interpMode=RCIM_Cubic, leaveTangent=6.1645, tangentMode=RCTM_User, time=0.0035, value=0.4903}, <at 0.9786 1.29>]}
       ZCurve {keys [
         <at -0.0011 1.0067 RCIM_Cubic>
         {arriveTangent=-1.635, interpMode=RCIM_Cubic, leaveTangent=-1.635, tangentMode=RCTM_User, time=1.0016, value=0.28}
        ]}
      }})) {enabled=false}
   /Game/Effects/Niagara/Modules/Materials/MaterialPanner(Offset:NiagaraFloat=&Add_Float(A=$Emitter.Age, B=$Particles.Initial.MaterialRandom), Speed=-20)
   &DynamicMaterialParameters(
    Index 0 Param 1=&FloatFromCurve(FloatCurve=DataInterface() {propertyValues {Curve {keys [<at 0.4984 -0.0132>,<at 1.0002 0.9984>]}}})
    Index 0 Param 3=3
    Index 0 Param 4=1.38
    Module.Param0WriteEnabled=true
    Module.Param1WriteEnabled=true
    Module.Param2WriteEnabled=true
    Module.Param3WriteEnabled=true
   ) {enabled=false}
   /Game/Effects/Niagara/Modules/Light/FN_Light_v5/FN_Light(
    Position Offset:Vector3f=<-10 35 10>
    Scale Color:NiagaraFloat=&Multiply_Float(A=500, B=&VectorLength(Vector=$Particles.Scale))
   ) {enabled=false}
   &AddVelocity() {enabled=false}
  ]
  rendererClasses [&NiagaraLightRendererProperties, &NiagaraMeshRendererProperties]
  renderers [
   {rendererClass=&NiagaraLightRendererProperties, rendererIndex=0}
   {rendererClass=&NiagaraMeshRendererProperties, rendererIndex=1}
  ]
 }
 Swirls_Mesh {
  emitterUpdateScript [
   &EmitterScalability(
    Module.bApplySpawnCountScale=false
    Module.bApplySpawnCountScaleByDistanceFraction=false
    Module.bUseMaxDistance=false
    Module.bUseMinDistance=false
   ) {enabled=false}
   &EmitterLifeCycle(MaxLoopCount=1, NextLoopDuration=1) {enabled=false}
   &EmitterState(Enable Distance Culling=true, Module.bUseMaxDistance=true, Module.bUseMinDistance=false, Scalability Mode=::NewEnumerator1)
   &SpawnRate(SpawnRate=10, Use Spawn Probability=false)
   &SpawnBurst_Instantaneous(Spawn Count=1, Use Spawn Probability=false)
  ]
  particleSpawnScript [
   &InitializeParticle(
    Lifetime=0.5
    Mesh Scale=&UniformRangedVector(
     Minimum=&UniformRangedVector(Minimum=<1 6 1>, Maximum=<1 3 1>)
     Maximum=&UniformRangedVector(Minimum=<1 -3 1>, Maximum=<1 -6 1>)
    )
    Module.Write Color=true
    Module.Write Lifetime=true
    Module.Write Mass=true
    Module.Write Material Random=true
    Module.Write Position=true
    Module.Write Scale=true
    Module.Write SpriteRotation=false
    Module.Write SpriteSize=false
    Position=/Niagara/Functions/Localspace/SimulationPosition()
    Sprite Size=<25 25>
   )
   &DynamicMaterialParameters(Module.Param0WriteEnabled=true, Module.Param1WriteEnabled=true, Module.Param2WriteEnabled=true, Module.Param3WriteEnabled=true) {enabled=false}
   &AddVelocity(Velocity=<0 0 5>)
   /SetVariables(Particles.RotationRate:NiagaraFloat=&UniformRangedFloat(Minimum=0.5, Maximum=1)) {isSetParameters=true}
   &InitialMeshOrientation(Orientation Vector=<1 180 0>, Use Orientation Vector=true, Use Rotation Vector=false)
   &BoxLocation(
    Box Size=&UniformRangedVector(Minimum=<0 0 0>, Maximum=<0 0 40>)
    MaskSpawnGroup=false
    Module.Override Seed=false
    Rotate Box=false
    Surface Only=false
   )
  ]
  particleUpdateScript [
   &SolveForcesAndVelocity(Clamp Velocity=false, Limit Acceleration=false)
   &ParticleState()
   &ScaleColor(
    Scale Alpha=&FloatFromCurve(FloatCurve=DataInterface() {propertyValues {Curve {keys [
         {arriveTangent=-5.932, interpMode=RCIM_Cubic, leaveTangent=-5.932, tangentMode=RCTM_User, time=-0.0046, value=1.0113}
         {arriveTangent=-0.9564, interpMode=RCIM_Cubic, leaveTangent=-0.9564, tangentMode=RCTM_User, time=0.2132, value=0.2272}
         <at 1.0011 0.002>
        ]}}})
    Scale RGB=&MakeVectorFromLinearColorRGB(Color=<rgba 1 1 1 1>)
    ScaleA=true
    ScaleRGB=true
   )
   &FN_PerParticleLOD(Enable SpriteSize Distance Scaling=true)
   &ScaleMeshSize(Scale Factor=<0.7 1 0.7>)
   &ScaleMeshSize(Scale Factor=&VectorFromFloat(Value=&FloatFromCurve(FloatCurve=DataInterface() {propertyValues {Curve {keys [{arriveTangent=1.4749, interpMode=RCIM_Cubic, leaveTangent=1.4749, tangentMode=RCTM_User, time=0, value=0.5}, <at 1.0629 2>]}}}))) {enabled=false, module=ScaleMeshSize001}
   /Niagara/Modules/Update/Orientation/V2/OrientMeshToVector(Override Quat=false, UseOptionalSideAxis=false, UseOptionalUpAxis=false)
   /Niagara/Modules/Update/Orientation/V2/UpdateMeshOrientation(Rotation Vector:Vector3f=<0 0 -4>)
   &DynamicMaterialParameters(
    Index 0 Param 1=&FloatFromCurve(
     FloatCurve=DataInterface() {propertyValues {Curve {keys [<at 0 0>,<at 0.1 1>,<at 0.2956 0.4308>,<at 0.5863 0.1532>,<at 0.99 0>]}}}
     Scale Curve=0.25
    )
    Module.Param0WriteEnabled=true
    Module.Param1WriteEnabled=true
    Module.Param2WriteEnabled=true
    Module.Param3WriteEnabled=true
   ) {module=DynamicMaterialParameters001}
  ]
  rendererClasses [&NiagaraMeshRendererProperties]
  renderers [{rendererClass=&NiagaraMeshRendererProperties, rendererIndex=0}]
 }
}
summary {
 systemName=NS_BottleCap_Challenge_CapSpin
 emitters [
  {emitterName=Swirls_Mesh, rendererClasses [&NiagaraMeshRendererProperties]}
  {emitterName=Flash, rendererClasses [&NiagaraLightRendererProperties, &NiagaraMeshRendererProperties]}
  {emitterName=Big_Glow_Mesh, rendererClasses [&NiagaraMeshRendererProperties]}
 ]
 userVariables [
  User.bIsFrontEndPreview {type=NiagaraBool}
  User.PlayerSpeed {type=NiagaraFloat, defaultValue=0, description=""}
  User.SkeletalMesh {
   type=NiagaraDataInterfaceSkeletalMesh
   defaultValue {
    struct=NiagaraExt_VariableValue_DataInterface
    value {dataInterface=&self:NiagaraDataInterfaceSkeletalMesh_0}
   }
  }
 ]
}
